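# Physics asset for the Textile Pup Male two-handed pickaxe, giving its shoe bone chain physics shapes, likely for dangling secondary motion.
# Four capsule bodies cover the handle and the ankle, mid and toe shoe joints, linked in series by three constraints.
# The ankle is held tightly to the handle in swing and twist, while the mid and toe joints bend widely with little twist.
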PhysicsAsset /BRCosmetics/Weapons/FORT_Melee/Pickaxe_Textile_Pup_Male/Meshes/Textile_Pup_Male_Physics
bodies {
 Melee_TwoHanded_Handle [{shapeType=Capsule, length=5, radius=2, shapeName=Melee_TwoHanded_Handle_capsule}]
 dyn_shoe_ankle_jnt [{shapeType=Capsule, length=3, radius=4, shapeName=dyn_shoe_ankle_jnt_capsule}]
 dyn_shoe_mid_jnt [{shapeType=Capsule, length=2, radius=7, shapeName=dyn_shoe_mid_jnt_capsule}]
 dyn_shoe_toe_jnt [{shapeType=Capsule, length=3, radius=7, shapeName=dyn_shoe_toe_jnt_capsule, center=<0.0001 0 -1.4565>}]
}
constraints [
 {bone1Name=dyn_shoe_ankle_jnt, bone2Name=Melee_TwoHanded_Handle, swing1LimitDegrees=5, swing2LimitDegrees=5, twistLimitDegrees=5}
 {bone1Name=dyn_shoe_mid_jnt, bone2Name=dyn_shoe_ankle_jnt, swing1LimitDegrees=45, swing2LimitDegrees=45, twistLimitDegrees=5}
 {bone1Name=dyn_shoe_toe_jnt, bone2Name=dyn_shoe_mid_jnt, swing1LimitDegrees=45, swing2LimitDegrees=45, twistLimitDegrees=5}
]
